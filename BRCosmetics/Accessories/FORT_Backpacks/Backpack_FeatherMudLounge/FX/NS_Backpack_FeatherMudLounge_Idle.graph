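# Idle Niagara effect for the FeatherMudLounge back bling: one sprite emitter of short-lived stars that follow the backpack's skeletal mesh.
# Stars spawn probabilistically, take time-of-day colors, and pop up in size then shrink away.
# A second mesh sample, which leaves position untouched, anchors a region mask.
# User.CrownTurnedOff scales spawn rate up to double and, through a comparison against 1, switches alpha between that mask and full visibility.
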NiagaraSystem /BRCosmetics/Accessories/FORT_Backpacks/Backpack_FeatherMudLounge/FX/NS_Backpack_FeatherMudLounge_Idle
dependencies {
 usedDataInterfaces [&NiagaraDataInterfaceCurve, &NiagaraDataInterfaceSkeletalMesh]
 usedDynamicInputs []
 usedModules [
  &self:SystemSpawnScript.SystemScriptSource.SystemScriptGraph.NiagaraNodeAssignment_0.SetVariables_#1
  &SystemState
 ]
}
emitters {Stars {
  emitterUpdateScript [
   &EmitterState()
   &SpawnRate(
    Spawn Probability=0.6
    SpawnRate=&Lerp_Float(A=2.3, B=4.6, Alpha=$User.CrownTurnedOff)
    Use Spawn Probability=true
   )
  ]
  particleSpawnScript [
   &InitializeParticle(
    Color=&FNTimeOfDayColor(Night Color=<rgba 1 0.6005 0 0.8>, Morning Color=<rgba 2 1.2009 0 2>, Day Color=<rgba 2 1.2009 0 2>, Evening Color=<rgba 2 1.2009 0 0.9>)
    Lifetime Max=1.3
    Lifetime Mode=::NewEnumerator1
    Sprite Size Mode=::NewEnumerator1
    Uniform Sprite Size Max=25
    Uniform Sprite Size Min=15
    UsePositionOffset=false
    Write Material Random=true
   )
   &CameraOffset(Camera Offset Amount=15)
   &Color(Color=&MultiplyLinearColor(
     Linear Color=$Particles.Color
     Scale Factor=&MakeVector4_FromLinearColor(LinearColor=&FNTimeOfDayColor(Night Color=<rgba 0.3611 0.3761 0.6979 1>, Morning Color=<rgba 1 0.717 0.7469 1>, Day Color=<rgba 1 1 1 1>, Evening Color=<rgba 1 1 1 1>))
    ))
  ]
  particleUpdateScript [
   &SkeletalMeshLocation(
    MaskSpawnGroup=false
    Mesh Sampling Type=::NewEnumerator7
    OffsetAlongNormal=false
    OverrideRandomSeed=false
    Skeletal Mesh=$System.Skeletal Mesh
   )
   &SkeletalMeshLocation(Bone Sampling Mode=::NewEnumerator4, MaskSpawnGroup=false, Skeletal Mesh=$System.Skeletal Mesh, Write To Particle Position=false) {module=SkeletalMeshLocation001}
   &RegionMask(
    Offset:Vector3f=<0 0 -8.3561>
    Region Dimensions:Vector3f=<100 100 40>
    Region Origin:Vector3f=$Particles.SkeletalMeshLocation001.SampledPosition
    Region Type:ENiagaraDebugDrawShape=::NewEnumerator1
    UseOffset=true
    UseRotation=false
   )
   &ParticleState()
   &ScaleSpriteSize(Uniform Curve Sprite Scale=DataInterface() {propertyValues {Curve {keys [
        {interpMode=RCIM_Cubic, leaveTangent=11.6533, tangentMode=RCTM_Break, time=0, value=0}
        {interpMode=RCIM_Cubic, leaveTangent=-4.0415, tangentMode=RCTM_Break, time=0.2, value=1}
        {tangentMode=RCTM_User, time=1, value=0}
       ]}}})
   &ScaleColor(
    Scale Alpha=&MakeCustomFloatFromBool(
     Bool=&SetBoolByFloatComparison(A=$User.CrownTurnedOff, B=1, Comparison Type=::NewEnumerator4)
     True Float=$Particles.RegionMask.Float
     False Float=1
    )
    ScaleA=true
    ScaleRGB=true
   )
  ]
  rendererClasses [&NiagaraSpriteRendererProperties]
  renderers [{rendererClass=&NiagaraSpriteRendererProperties}]
 }}
summary {
 systemName=NS_Backpack_FeatherMudLounge_Idle
 emitters [{emitterName=Stars, rendererClasses [&NiagaraSpriteRendererProperties]}]
 userVariables [User.CrownTurnedOff {type=NiagaraFloat, defaultValue=0, description=""}]
}
