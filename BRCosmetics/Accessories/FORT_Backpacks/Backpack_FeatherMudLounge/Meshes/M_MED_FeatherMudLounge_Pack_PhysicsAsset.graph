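# Physics asset for the FeatherMudLounge back bling backpack mesh in the cosmetic accessories set.
# It defines a single capsule body on the pack bone, offset and rotated to lie along the pack's shape.
# There are no constraints, so it gives the pack one simple collision volume with no jointed parts.
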PhysicsAsset /BRCosmetics/Accessories/FORT_Backpacks/Backpack_FeatherMudLounge/Meshes/M_MED_FeatherMudLounge_Pack_PhysicsAsset
bodies {pack [{shapeType=Capsule, length=10.7143, radius=6.6252, center=<0.2383 -4.1604 -2.1088>, rotation=<rot 0 88.9934 -174.7215>}]}
constraints []
