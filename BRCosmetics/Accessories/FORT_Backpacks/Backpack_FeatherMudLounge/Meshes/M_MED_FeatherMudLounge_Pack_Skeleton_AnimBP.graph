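# Animation Blueprint for the FeatherMudLounge backpack skeleton that adds procedural idle sway to the pack.
# Each update, the event graph builds three X/Z offset vectors (Main, Crown, Trapezoid) from real-time sine waves at differing frequencies and amplitudes.
# In the AnimGraph, only an additive bone-space move of dyn_crown by the Trapezoid offset, partially weighted, reaches the output pose.
# The dyn_main node and the other dyn_crown node, which read Main and Crown, are left unconnected.
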AnimBlueprint /BRCosmetics/Accessories/FORT_Backpacks/Backpack_FeatherMudLounge/Meshes/M_MED_FeatherMudLounge_Pack_Skeleton_AnimBP
target_skeleton=&/M_MED_FeatherMudLounge_Pack_Skeleton
graphs [
 AnimGraph {nodes [
   "Transform (Modify) Bone - Bone: dyn_main"(Alpha=1, Translation=Get Translate Main() -> (Translate Main:Vector), Rotation="0,0,0", Scale="1,1,1") {translationMode=BMM_Additive, translationSpace=BCS_BoneSpace, alphaBoolBlend {blendOption=Linear}, boneToModify {boneName=dyn_main}}
   "Transform (Modify) Bone - Bone: dyn_crown"(Alpha=0.4, Translation=Get Translate Crown() -> (Translate Crown:Vector), Rotation="0,0,0", Scale="1,1,1") {translationMode=BMM_Additive, translationSpace=BCS_BoneSpace, alphaBoolBlend {blendOption=Linear}, boneToModify {boneName=dyn_crown}}
   n1 = "Transform (Modify) Bone - Bone: dyn_crown"(
    ComponentPose=Local To Component(LocalPose=Input Pose())
    Alpha=0.6
    Translation=Get Translate Trapezoid() -> (Translate Trapezoid:Vector)
    Rotation="0,0,0"
    Scale="1,1,1"
   ) {translationMode=BMM_Additive, translationSpace=BCS_BoneSpace, alphaBoolBlend {blendOption=Linear}, boneToModify {boneName=dyn_crown}}
   Output Pose(Result=Component To Local(ComponentPose=@n1))
  ]}
 EventGraph {nodes [
   Event BlueprintUpdateAnimation()
   > n1 = Sequence!(execute) -> (then_0, then_1, then_2:Exec)
   n2 = GetRealTimeSeconds()
   n3 = "Sin (Radians)"(A=float * float(A=float * float(A=1, B=-1.2), B=@n2) {is=title K2Node_PromotableOperator})
   n4 = "Sin (Radians)"(A=float * float(A=float * float(A=0.2, B=-2), B=@n2))
   Set Translate Crown(execute=@n1.then_2, Translate Crown:Vector=MakeVector(X=float / float(A=@n3, B=-1.3), Y=0, Z=float / float(A=@n4, B=4))) -> (Output_Get:Vector)
   n5 = GetRealTimeSeconds()
   n6 = "Sin (Radians)"(A=float * float(A=float * float(A=1, B=-1.6), B=@n5))
   n7 = "Sin (Radians)"(A=float * float(A=float * float(A=0.4, B=3), B=@n5))
   Set Translate Trapezoid(execute=@n1.then_0, Translate Trapezoid:Vector=MakeVector(X=float / float(A=@n6, B=2), Y=0, Z=float / float(A=@n7, B=1.5))) -> (Output_Get:Vector)
   n8 = GetRealTimeSeconds()
   n9 = "Sin (Radians)"(A=float * float(A=float * float(A=1, B=-1.4), B=@n8))
   n10 = "Sin (Radians)"(A=float * float(A=float * float(A=0.8, B=-0.8), B=@n8))
   Set Translate Main(execute=@n1.then_1, Translate Main:Vector=MakeVector(X=float / float(A=@n9, B=4.3), Y=0, Z=float / float(A=@n10, B=6.3))) -> (Output_Get:Vector)
   TryGetPawnOwner()
   n11 = float * float(A=1, B=1.1)
   float / float(A="Sin (Radians)"(A=float * float(A=@n11, B=@n2)), B=20)
   n12 = float * float(A=1, B=1.3)
   float / float(A="Sin (Radians)"(A=float * float(A=@n12, B=@n8)), B=20)
   n13 = float * float(A=0.7, B=1.3)
   float / float(A="Sin (Radians)"(A=float * float(A=@n13, B=@n5)), B=4)
  ]}
]
variables [
 Translate Crown {type=Vector}
 Translate Main {type=Vector}
 Translate Trapezoid {type=Vector}
]
